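# Animation Blueprint for the Sand Mansion gadget prop skeleton, hiding the prop until its emote starts and then playing the CMM prop loop.
# An instant bool blend picks either the loop or a ref pose with root scaled to zero, then feeds DefaultSlot for montage overrides.
# Each update copies bHasEmoteStarted from the owning B_PersistentProp_Skeletal until it turns true, so the flag latches.
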AnimBlueprint /BRCosmetics/Animation/Props/Gadgets_SandMansion_Mansion/Meshes/SandMansion_Mansion_Gadget_Skeleton_AnimBP
target_skeleton=&/SandMansion_Mansion_Gadget_Skeleton
graphs [
 AnimGraph {nodes [
   n1 = "Transform (Modify) Bone - Bone: root"!(ComponentPose=Local To Component(LocalPose=Local Space Ref Pose()), Alpha, Scale="0.000,0.000,0.000") -> (Pose) {scaleMode=BMM_Replace, alphaBoolBlend {blendOption=Linear}, boneToModify {boneName=root}, scale=<0 0 0>}
   n2 = Component To Local(ComponentPose=@n1)
   n3 = Sequence Player 'Emote_SandMansion_CMM_Prop_Loop'() {sequence=/BRCosmetics/Animation/Game/MainPlayer/Emotes/Sandmansion/CMM/Emote_SandMansion_CMM_Prop_Loop}
   n4 = Get bHasEmoteStarted() -> (bHasEmoteStarted:bool)
   Output Pose(Result=Slot 'DefaultSlot'(Source=Blend Poses by bool(BlendPose_0=@n3, BlendPose_1=@n2, BlendTime_0=0, BlendTime_1=0, bActiveValue=@n4) {blendType=Linear}))
  ]}
 EventGraph {nodes [
   Event BlueprintUpdateAnimation()
   > Branch(Condition=NOT Boolean(A=Get bHasEmoteStarted()))
   > n1 = Cast To B_PersistentProp_Skeletal!(execute, Object=GetOwningActor()) -> (then, CastFailed, AsB Persistent Prop Skeletal:B Persistent Prop Skeletal Object Reference) {is=title K2Node_DynamicCast}
   > Set bHasEmoteStarted(bHasEmoteStarted:bool=Get bHasEmoteStarted!(self:B Persistent Prop Skeletal Object Reference=@n1) -> (bHasEmoteStarted)) -> (Output_Get:bool)
   TryGetPawnOwner()
  ]}
]
variables [bHasEmoteStarted {type=boolean}]
